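# Conversation database for the Jenkins NPC, apparently a test transient-quest dialogue in Battlepass Season 15.
# The Conversation.Entry.Jenkins entry point leads to a Speech node, which leads to a GrantStaticQuest node.
# That node carries an "ACCEPT!" choice and grants the quest_jenkinstest_ind_n01_q01 quest item, ending the conversation.
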ConversationDatabase /BattlepassS15/TransientQuests/Conversations/Conversation_Jenkins
entry_points [{tag=Conversation.Entry.Jenkins, destinations [#1]}]
nodes [
 ConversationEntryPointNode_7 {
  id=#1
  class=ConversationEntryPointNode
  outputs [#2]
  properties {entryTag {tagName=Conversation.Entry.Jenkins}}
 }
 CTB_SaySomething_C_6 {id=#2, class=Speech_C, outputs [#3], properties {nodeName=""}}
 GrantStaticQuest_C_1 {
  id=#3
  class=GrantStaticQuest_C
  outputs []
  properties {questToGrantPtr=/BattlepassS15/Items/QuestItems/JenkinsTest/quest_jenkinstest_ind_n01_q01}
  sub_nodes [ConversationChoiceNode_0 {class=ConversationChoiceNode, properties {defaultChoiceDisplayText="ACCEPT!"}}]
 }
]
speakers [{}, {participant=Conversation.Participant.NPC}]
